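# Sound cue for large quadruped scuff foley in CosmeticCompanions, classed as a pet creature sound.
# A quality-level root routes its first two levels to a random pick of ten hard scuff waves and its last two to a pool of just the first five, all equally weighted.
# The cue plays quietly and pitched up at zero priority, using companion foley attenuation and a quadruped footstep concurrency set.
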SoundCue /CosmeticCompanions/Assets/Audio_Shared/Foley/Quadraped/Large/SC_MImosa_Quadraped_Lrg_Scuff
root=QualityLevel_0
nodes [
 {id=QualityLevel_0, children [Random_2,Random_2,Random_4,Random_4]}
 {
  id=Random_2
  children [
   WavePlayer_17
   WavePlayer_18
   WavePlayer_19
   WavePlayer_20
   WavePlayer_21
   WavePlayer_22
   WavePlayer_23
   WavePlayer_24
   WavePlayer_25
   WavePlayer_26
  ]
  properties {weights [1,1,1,1,1,1,1,1,1,1]}
 }
 {id=Random_4, children [WavePlayer_37, WavePlayer_38, WavePlayer_39, WavePlayer_40, WavePlayer_41], properties {weights [1,1,1,1,1]}}
 {
  id=WavePlayer_17
  properties {soundWaveAssetPtr=/CosmeticCompanions/Assets/Audio_Shared/Foley/Quadraped/Large/Wavs/Scuff/Mimosa_Quadraped_Lrg_Scuff_Hard_01}
 }
 {
  id=WavePlayer_18
  properties {soundWaveAssetPtr=/CosmeticCompanions/Assets/Audio_Shared/Foley/Quadraped/Large/Wavs/Scuff/Mimosa_Quadraped_Lrg_Scuff_Hard_02}
 }
 {
  id=WavePlayer_19
  properties {soundWaveAssetPtr=/CosmeticCompanions/Assets/Audio_Shared/Foley/Quadraped/Large/Wavs/Scuff/Mimosa_Quadraped_Lrg_Scuff_Hard_03}
 }
 {
  id=WavePlayer_20
  properties {soundWaveAssetPtr=/CosmeticCompanions/Assets/Audio_Shared/Foley/Quadraped/Large/Wavs/Scuff/Mimosa_Quadraped_Lrg_Scuff_Hard_04}
 }
 {
  id=WavePlayer_21
  properties {soundWaveAssetPtr=/CosmeticCompanions/Assets/Audio_Shared/Foley/Quadraped/Large/Wavs/Scuff/Mimosa_Quadraped_Lrg_Scuff_Hard_05}
 }
 {
  id=WavePlayer_22
  properties {soundWaveAssetPtr=/CosmeticCompanions/Assets/Audio_Shared/Foley/Quadraped/Large/Wavs/Scuff/Mimosa_Quadraped_Lrg_Scuff_Hard_06}
 }
 {
  id=WavePlayer_23
  properties {soundWaveAssetPtr=/CosmeticCompanions/Assets/Audio_Shared/Foley/Quadraped/Large/Wavs/Scuff/Mimosa_Quadraped_Lrg_Scuff_Hard_07}
 }
 {
  id=WavePlayer_24
  properties {soundWaveAssetPtr=/CosmeticCompanions/Assets/Audio_Shared/Foley/Quadraped/Large/Wavs/Scuff/Mimosa_Quadraped_Lrg_Scuff_Hard_08}
 }
 {
  id=WavePlayer_25
  properties {soundWaveAssetPtr=/CosmeticCompanions/Assets/Audio_Shared/Foley/Quadraped/Large/Wavs/Scuff/Mimosa_Quadraped_Lrg_Scuff_Hard_09}
 }
 {
  id=WavePlayer_26
  properties {soundWaveAssetPtr=/CosmeticCompanions/Assets/Audio_Shared/Foley/Quadraped/Large/Wavs/Scuff/Mimosa_Quadraped_Lrg_Scuff_Hard_10}
 }
 {
  id=WavePlayer_37
  properties {soundWaveAssetPtr=&Mimosa_Quadraped_Lrg_Scuff_Hard_01}
 }
 {
  id=WavePlayer_38
  properties {soundWaveAssetPtr=&Mimosa_Quadraped_Lrg_Scuff_Hard_02}
 }
 {
  id=WavePlayer_39
  properties {soundWaveAssetPtr=&Mimosa_Quadraped_Lrg_Scuff_Hard_03}
 }
 {
  id=WavePlayer_40
  properties {soundWaveAssetPtr=&Mimosa_Quadraped_Lrg_Scuff_Hard_04}
 }
 {
  id=WavePlayer_41
  properties {soundWaveAssetPtr=&Mimosa_Quadraped_Lrg_Scuff_Hard_05}
 }
]
properties {
 attenuationSettings=/Game/Sounds/Attenuation/AS_Companion_Foley
 maxDistance=1200
 pitchMultiplier=1.5
 priority=0
 soundClassObject=&Fort_Creatures_Pets
 volumeMultiplier=0.3
 concurrencySet [/CosmeticCompanions/Assets/Audio_Shared/Foley/Quadraped/CONC_Mimosa_Quadraped_Footstep]
}
