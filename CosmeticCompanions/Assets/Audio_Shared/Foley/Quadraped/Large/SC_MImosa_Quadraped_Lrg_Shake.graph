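# Sound cue for the shake foley of large quadruped cosmetic companions (Mimosa), using SitPlant shake recordings.
# A Random node picks one of three equally weighted wave variations on each play.
# Output is routed to the pets creature sound class with boosted volume and companion foley attenuation.
# It shares the quadruped footstep concurrency set, so shakes are limited alongside any other sounds in that set.
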SoundCue /CosmeticCompanions/Assets/Audio_Shared/Foley/Quadraped/Large/SC_MImosa_Quadraped_Lrg_Shake
root=Random_2
nodes [
 {id=Random_2, children [WavePlayer_17, WavePlayer_18, WavePlayer_19], properties {weights [1,1,1]}}
 {id=WavePlayer_17, properties {soundWaveAssetPtr=/CosmeticCompanions/Assets/Quadruped/SitPlant/Audio/Wavs/Foley/SitPlant_Shake_01}}
 {id=WavePlayer_18, properties {soundWaveAssetPtr=/CosmeticCompanions/Assets/Quadruped/SitPlant/Audio/Wavs/Foley/SitPlant_Shake_02}}
 {id=WavePlayer_19, properties {soundWaveAssetPtr=/CosmeticCompanions/Assets/Quadruped/SitPlant/Audio/Wavs/Foley/SitPlant_Shake_03}}
]
properties {
 attenuationSettings=/Game/Sounds/Attenuation/AS_Companion_Foley
 maxDistance=1200
 soundClassObject=&Fort_Creatures_Pets
 volumeMultiplier=2
 concurrencySet [/CosmeticCompanions/Assets/Audio_Shared/Foley/Quadraped/CONC_Mimosa_Quadraped_Footstep]
}
